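# Looping sound cue for the Grumpy beam in the Figment S08 Dopey event sounds, on the map-wide events sound class.
# A random node picks one of two equally weighted looping beam wave variants.
# Attenuation is overridden with a short falloff and listener focus that lowers volume off-axis.
# Concurrency is overridden to cap voices at four, stopping the farthest before blocking new ones.
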SoundCue /Figment_S08/Athena/Sounds/Events/Dopey/Grumpy/Grumpy_Beam_Loop_Cue
root=Random_1
nodes [
 {id=Random_1, children [WavePlayer_4,WavePlayer_5], properties {weights [1,1]}}
 {id=WavePlayer_4, properties {bLooping=true, soundWaveAssetPtr=&/Grumpy_Beam_LoopA_04}}
 {id=WavePlayer_5, properties {bLooping=true, soundWaveAssetPtr=&/Grumpy_Beam_LoopA_03}}
]
properties {
 bOverrideAttenuation=true
 bOverrideConcurrency=true
 maxDistance=356
 soundClassObject=/Figment_S08/Athena/Sounds/Events/SoundClasses/Fort_Events_Mapwide
 volumeMultiplier=1
 attenuationOverrides {
  bEnableFocusInterpolation=true
  bEnableListenerFocus=true
  falloffDistance=100
  focusAzimuth=10
  nonFocusAzimuth=30
  nonFocusVolumeAttenuation=0.5
  stereoSpread=256
  attenuationShapeExtents=<256 156>
 }
 concurrencyOverrides {maxCount=4, resolutionRule=StopFarthestThenPreventNew}
}
