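# Sound cue for the Grumpy dropping sound in the Dopey event audio, on the map-wide events sound class with boosted volume.
# A single wave passes through a modulator, the borrowed Snow attenuation asset and a mixer with an empty second input.
# A fixed delay at the root holds playback back by several seconds.
# Overridden concurrency caps simultaneous instances and stops the oldest.
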SoundCue /Figment_S08/Athena/Sounds/Events/Dopey/Grumpy/Grumpy_Dropping_Cue
root=Delay_0
nodes [
 {
  id=Attenuation_0
  children [Modulator_1]
  properties {attenuationSettings=/Figment_S08/Athena/Sounds/Events/Dopey/Snow/ATT_Snow_MLTI_Close_Att}
 }
 {id=Delay_0, children [Mixer_0], properties {delayMax=3.75, delayMin=3.75}}
 {id=Mixer_0, children [Attenuation_0, null], properties {inputVolume [1,1]}}
 {id=Modulator_1, children [WavePlayer_3]}
 {id=WavePlayer_3, properties {soundWaveAssetPtr=&/Grumpy_Dropping_01}}
]
properties {
 bOverrideConcurrency=true
 maxDistance=22500
 soundClassObject=/Figment_S08/Athena/Sounds/Events/SoundClasses/Fort_Events_Mapwide
 volumeMultiplier=2
 attenuationOverrides {falloffDistance=3000, attenuationShapeExtents {x=0}}
 concurrencyOverrides {maxCount=4, resolutionRule=StopOldest}
}
