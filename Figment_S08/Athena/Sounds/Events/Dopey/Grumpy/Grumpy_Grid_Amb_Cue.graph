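# Looping ambient sound cue for the Grumpy grid in the Dopey event, judging by its name and path.
# A single looping wave player plays a platform loop wave, routed through the map-wide events sound class.
# Inline attenuation uses a capsule shape with distance falloff, and inline concurrency caps simultaneous copies by stopping the oldest.
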SoundCue /Figment_S08/Athena/Sounds/Events/Dopey/Grumpy/Grumpy_Grid_Amb_Cue
root=WavePlayer_2
nodes [{id=WavePlayer_2, properties {bLooping=true, soundWaveAssetPtr=/Game/Athena/Sounds/Events/Dopey/Grumpy/Grumpy_Platform_Loop_01}}]
properties {
 bOverrideAttenuation=true
 bOverrideConcurrency=true
 maxDistance=4048
 soundClassObject=/Figment_S08/Athena/Sounds/Events/SoundClasses/Fort_Events_Mapwide
 volumeMultiplier=1
 attenuationOverrides {attenuationShape=Capsule, falloffDistance=2048, attenuationShapeExtents=<2000 768>}
 concurrencyOverrides {maxCount=4, resolutionRule=StopOldest}
}
