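# Sound cue for the Grumpy panel completion sound in the Dopey event content, routed to the mapwide events sound class.
# A mixer plays two layers at once: a modulated close wave and a distant wave.
# Each layer has its own Snow close or distant attenuation asset, so the blend shifts with listener distance.
# A concurrency override caps simultaneous instances, stopping the oldest first.
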SoundCue /Figment_S08/Athena/Sounds/Events/Dopey/Grumpy/Grumpy_Panel_Complete_Cue
root=Mixer_0
nodes [
 {
  id=Attenuation_0
  children [Modulator_1]
  properties {attenuationSettings=/Figment_S08/Athena/Sounds/Events/Dopey/Snow/ATT_Snow_MLTI_Close_Att}
 }
 {
  id=Attenuation_1
  children [WavePlayer_2]
  properties {attenuationSettings=/Figment_S08/Athena/Sounds/Events/Dopey/Snow/ATT_Snow_MLTI_Distant_Att}
 }
 {id=Mixer_0, children [Attenuation_0, Attenuation_1], properties {inputVolume [1,1]}}
 {id=Modulator_1, children [WavePlayer_3]}
 {id=WavePlayer_2, properties {soundWaveAssetPtr=&/Grumpy_Panel_Complete_Distant_01}}
 {id=WavePlayer_3, properties {soundWaveAssetPtr=&/Grumpy_Panel_Complete_01}}
]
properties {
 bOverrideConcurrency=true
 maxDistance=2097152
 soundClassObject=/Figment_S08/Athena/Sounds/Events/SoundClasses/Fort_Events_Mapwide
 volumeMultiplier=2
 attenuationOverrides {falloffDistance=3000, attenuationShapeExtents {x=0}}
 concurrencyOverrides {maxCount=4, resolutionRule=StopOldest}
}
